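# Physics asset for the Silver Trophy gadget prop, providing collision and joints for its four hanging ribbons.
# A compound base body of capsules and boxes approximates the trophy.
# Left and right, front and back ribbons are each a five-box bone chain rooted on the base.
# Every joint shares one setup with free swing and a narrow twist limit, so the ribbons can dangle and flex.
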PhysicsAsset /BRCosmetics/Animation/Props/Silver_Trophy/Meshes/Silver_Trophy_Gadget_Physics
bodies {
 LB_Ribbon_01 [{shapeType=Box, extentX=12.2383, extentY=2.8658, extentZ=1.3103, center=<6.0758 -0.0001 0>}]
 LB_Ribbon_02 [{shapeType=Box, extentX=9.5512, extentY=3.2738, extentZ=1.3103, center=<4.717 -0.0001 -0.0001>}]
 LB_Ribbon_03 [{shapeType=Box, extentX=10.1124, extentY=4.0825, extentZ=1.3103, center=<4.909 -0.0001 0>}]
 LB_Ribbon_04 [{shapeType=Box, extentX=9.0355, extentY=4.2322, extentZ=1.3103, center=<4.5568 -0.0001 0>}]
 LB_Ribbon_05 [{shapeType=Box, extentX=7.9507, extentY=4.2226, extentZ=1.3103, center=<3.8153 -0.0001 0>}]
 LF_Ribbon_01 [{shapeType=Box, extentX=12.2383, extentY=2.3346, extentZ=1.3103, center=<6.0758 -0.0001 0>}]
 LF_Ribbon_02 [{shapeType=Box, extentX=9.5512, extentY=3.2738, extentZ=1.3103, center=<4.717 -0.0001 -0.0001>}]
 LF_Ribbon_03 [{shapeType=Box, extentX=8.47, extentY=3.4707, extentZ=1.3103, center=<4.362 -0.0001 0>}]
 LF_Ribbon_04 [{shapeType=Box, extentX=9.0355, extentY=4.2322, extentZ=1.3103, center=<4.5568 -0.0001 0>}]
 LF_Ribbon_05 [{shapeType=Box, extentX=7.9507, extentY=4.2226, extentZ=1.3103, center=<3.8153 -0.0001 0>}]
 RB_Ribbon_01 [{shapeType=Box, extentX=10.54, extentY=2.8658, extentZ=1.3103, center=<5.3588 -0.0001 0>}]
 RB_Ribbon_02 [{shapeType=Box, extentX=13.5186, extentY=3.3416, extentZ=1.3103, center=<6.6157 -0.0001 0>}]
 RB_Ribbon_03 [{shapeType=Box, extentX=9.5123, extentY=4.0825, extentZ=1.3103, center=<4.761 -0.0001 0>}]
 RB_Ribbon_04 [{shapeType=Box, extentX=9.7818, extentY=4.2322, extentZ=1.3103, center=<4.9123 -0.0001 0>}]
 RB_Ribbon_05 [{shapeType=Box, extentX=7.9507, extentY=5.1621, extentZ=1.3103, center=<3.8499 -0.2875 0>}]
 RF_Ribbon_01 [{shapeType=Box, extentX=11.0038, extentY=2.3346, extentZ=1.3103, center=<5.4685 -0.0001 0>}]
 RF_Ribbon_02 [{shapeType=Box, extentX=13.0634, extentY=2.994, extentZ=1.3103, center=<6.5188 -0.0001 -0.0001>}]
 RF_Ribbon_03 [{shapeType=Box, extentX=8.47, extentY=3.9101, extentZ=1.3103, center=<4.362 -0.0001 0>}]
 RF_Ribbon_04 [{shapeType=Box, extentX=9.0355, extentY=4.0711, extentZ=1.3103, center=<4.5568 -0.0001 0>}]
 RF_Ribbon_05 [{shapeType=Box, extentX=7.9507, extentY=4.9745, extentZ=1.3103, center=<3.8153 -0.0001 0>}]
 base [
  {shapeType=Capsule, length=39.0814, radius=9.535, center=<0.0855 -0.3803 26.3295>, rotation=<rot 0.4229 -90 0.0553>}
  {shapeType=Capsule, length=53.7936, radius=10.0371, center=<-30.6877 19.8598 36.9925>, rotation=<rot -68.294 73.3499 -74.4679>}
  {shapeType=Capsule, length=53.6752, radius=10.2353, center=<-29.0607 -18.5491 36.2022>, rotation=<rot -67.5898 -74.0264 75.4729>}
  {shapeType=Capsule, length=11.4825, radius=18.8724, center=<-40.1119 -0.454 47.9564>, rotation=<rot 0 0 -90.4335>}
  {shapeType=Box, extentX=1.9231, extentY=27.1369, extentZ=5.4544, center=<0.0001 12.0378 42.3323>, rotation=<rot 0 0 -65.7861>}
  {shapeType=Box, extentX=1.9231, extentY=27.1369, extentZ=5.4544, center=<-0.1179 -10.4184 40.6001>, rotation=<rot 0 0 -124.227>}
 ]
}
constraints [
 {
  bone1Name=LF_Ribbon_01
  bone2Name=base
  swing1LimitDegrees=45
  swing1Motion=Free
  swing2LimitDegrees=45
  swing2Motion=Free
  twistLimitDegrees=25
 }
 {
  bone1Name=LF_Ribbon_02
  bone2Name=LF_Ribbon_01
  swing1LimitDegrees=45
  swing1Motion=Free
  swing2LimitDegrees=45
  swing2Motion=Free
  twistLimitDegrees=25
 }
 {
  bone1Name=LF_Ribbon_03
  bone2Name=LF_Ribbon_02
  swing1LimitDegrees=45
  swing1Motion=Free
  swing2LimitDegrees=45
  swing2Motion=Free
  twistLimitDegrees=25
 }
 {
  bone1Name=LF_Ribbon_04
  bone2Name=LF_Ribbon_03
  swing1LimitDegrees=45
  swing1Motion=Free
  swing2LimitDegrees=45
  swing2Motion=Free
  twistLimitDegrees=25
 }
 {
  bone1Name=LF_Ribbon_05
  bone2Name=LF_Ribbon_04
  swing1LimitDegrees=45
  swing1Motion=Free
  swing2LimitDegrees=45
  swing2Motion=Free
  twistLimitDegrees=25
 }
 {
  bone1Name=LB_Ribbon_01
  bone2Name=base
  swing1LimitDegrees=45
  swing1Motion=Free
  swing2LimitDegrees=45
  swing2Motion=Free
  twistLimitDegrees=25
 }
 {
  bone1Name=LB_Ribbon_02
  bone2Name=LB_Ribbon_01
  swing1LimitDegrees=45
  swing1Motion=Free
  swing2LimitDegrees=45
  swing2Motion=Free
  twistLimitDegrees=25
 }
 {
  bone1Name=LB_Ribbon_03
  bone2Name=LB_Ribbon_02
  swing1LimitDegrees=45
  swing1Motion=Free
  swing2LimitDegrees=45
  swing2Motion=Free
  twistLimitDegrees=25
 }
 {
  bone1Name=LB_Ribbon_04
  bone2Name=LB_Ribbon_03
  swing1LimitDegrees=45
  swing1Motion=Free
  swing2LimitDegrees=45
  swing2Motion=Free
  twistLimitDegrees=25
 }
 {
  bone1Name=LB_Ribbon_05
  bone2Name=LB_Ribbon_04
  swing1LimitDegrees=45
  swing1Motion=Free
  swing2LimitDegrees=45
  swing2Motion=Free
  twistLimitDegrees=25
 }
 {
  bone1Name=RB_Ribbon_01
  bone2Name=base
  swing1LimitDegrees=45
  swing1Motion=Free
  swing2LimitDegrees=45
  swing2Motion=Free
  twistLimitDegrees=25
 }
 {
  bone1Name=RB_Ribbon_02
  bone2Name=RB_Ribbon_01
  swing1LimitDegrees=45
  swing1Motion=Free
  swing2LimitDegrees=45
  swing2Motion=Free
  twistLimitDegrees=25
 }
 {
  bone1Name=RB_Ribbon_03
  bone2Name=RB_Ribbon_02
  swing1LimitDegrees=45
  swing1Motion=Free
  swing2LimitDegrees=45
  swing2Motion=Free
  twistLimitDegrees=25
 }
 {
  bone1Name=RB_Ribbon_04
  bone2Name=RB_Ribbon_03
  swing1LimitDegrees=45
  swing1Motion=Free
  swing2LimitDegrees=45
  swing2Motion=Free
  twistLimitDegrees=25
 }
 {
  bone1Name=RB_Ribbon_05
  bone2Name=RB_Ribbon_04
  swing1LimitDegrees=45
  swing1Motion=Free
  swing2LimitDegrees=45
  swing2Motion=Free
  twistLimitDegrees=25
 }
 {
  bone1Name=RF_Ribbon_01
  bone2Name=base
  swing1LimitDegrees=45
  swing1Motion=Free
  swing2LimitDegrees=45
  swing2Motion=Free
  twistLimitDegrees=25
 }
 {
  bone1Name=RF_Ribbon_02
  bone2Name=RF_Ribbon_01
  swing1LimitDegrees=45
  swing1Motion=Free
  swing2LimitDegrees=45
  swing2Motion=Free
  twistLimitDegrees=25
 }
 {
  bone1Name=RF_Ribbon_03
  bone2Name=RF_Ribbon_02
  swing1LimitDegrees=45
  swing1Motion=Free
  swing2LimitDegrees=45
  swing2Motion=Free
  twistLimitDegrees=25
 }
 {
  bone1Name=RF_Ribbon_04
  bone2Name=RF_Ribbon_03
  swing1LimitDegrees=45
  swing1Motion=Free
  swing2LimitDegrees=45
  swing2Motion=Free
  twistLimitDegrees=25
 }
 {
  bone1Name=RF_Ribbon_05
  bone2Name=RF_Ribbon_04
  swing1LimitDegrees=45
  swing1Motion=Free
  swing2LimitDegrees=45
  swing2Motion=Free
  twistLimitDegrees=25
 }
]
