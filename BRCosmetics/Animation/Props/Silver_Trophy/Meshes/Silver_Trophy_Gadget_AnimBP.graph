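# Animation Blueprint for the Silver Trophy gadget prop skeleton that adds physics-driven secondary motion to an incoming pose.
# The input pose passes through DefaultSlot into a RigidBody node, simulated in root-bone space with overridden gravity and scaled component acceleration, then to output.
# A second RigidBody branch with a cached pose and ribbon-bone layered blend is left unconnected to the output.
# The EventGraph holds only unwired default nodes.
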AnimBlueprint /BRCosmetics/Animation/Props/Silver_Trophy/Meshes/Silver_Trophy_Gadget_AnimBP
target_skeleton=&/Silver_Trophy_Gadget_Trophy_LOD0_Skeleton
graphs [
 AnimGraph {nodes [
   n1 = RigidBody(ComponentPose=Local To Component(LocalPose=Use cached pose 'InputCache'())) {
    bClampLinearTranslationLimitToRefPose=true
    bOverrideWorldGravity=true
    overridePhysicsAsset=&/Silver_Trophy_Gadget_Physics
    simulationSpace=BaseBoneSpace
    alphaBoolBlend {blendOption=Linear}
    baseBoneRef {boneName=root}
    componentAppliedLinearAccClamp=<5000 5000 5000>
    componentLinearAccScale=<0.25 0.25 0.45>
    overrideWorldGravity {z=-980}
   }
   Layered blend per bone(BlendPoses_0=Component To Local(ComponentPose=@n1)) {layerSetup [{branchFilters [
       {blendDepth=1, boneName=LF_Ribbon_01}
       {blendDepth=1, boneName=LB_Ribbon_01}
       {blendDepth=1, boneName=RB_Ribbon_01}
       {blendDepth=1, boneName=RF_Ribbon_01}
      ]}]}
   n2 = Input Pose()
   n3 = RigidBody(ComponentPose=Local To Component(LocalPose=Slot 'DefaultSlot'(Source=@n2))) {
    bClampLinearTranslationLimitToRefPose=true
    bOverrideWorldGravity=true
    overridePhysicsAsset=&Silver_Trophy_Gadget_Physics
    simulationSpace=BaseBoneSpace
    alphaBoolBlend {blendOption=Linear}
    baseBoneRef {boneName=root}
    componentAppliedLinearAccClamp=<5000 5000 5000>
    componentLinearAccScale=<0.25 0.25 0.45>
    overrideWorldGravity {z=-980}
   }
   Output Pose(Result=Component To Local(ComponentPose=@n3)) {layerGroup=None}
   Save cached pose 'InputCache'() {cacheName=InputCache}
   Use cached pose 'InputCache'()
  ]}
 EventGraph {nodes [
   TryGetPawnOwner()
   Event BlueprintUpdateAnimation()
  ]}
]
variables []
